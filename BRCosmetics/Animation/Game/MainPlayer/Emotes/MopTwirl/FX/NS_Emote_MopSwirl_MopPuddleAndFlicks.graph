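# Niagara effect for the MopTwirl (MopSwirl) emote that leaves a wet mop trail and flicks water off the mop.
# BlowingParticles fires sprite bursts at a sampled socket when User.BurstIndex changes, gated by system age, with count, direction and speed chosen per index.
# Trail spawns decals and sprites at User.SpawnCurve's rate, using a slope-position module for placement.
# User.bIsFrontEndPreview swaps which Trail spawn groups run and adds drift velocity.
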NiagaraSystem /BRCosmetics/Animation/Game/MainPlayer/Emotes/MopTwirl/FX/NS_Emote_MopSwirl_MopPuddleAndFlicks
dependencies {
 usedDataInterfaces [
  &NiagaraDataInterfaceArrayColor
  &NiagaraDataInterfaceArrayInt32
  &NiagaraDataInterfaceCollisionQuery
  &NiagaraDataInterfaceDebugDraw
  &NiagaraDataInterfaceCurve
  &NiagaraDataInterfaceArrayFloat
  &NiagaraDataInterfaceSpriteRendererInfo
  &NiagaraDataInterfaceSocketReader
 ]
 usedDynamicInputs []
 usedModules [
  &SystemState
  &BlendSpaceHelper
  &self:SystemSpawnScript.SystemScriptSource.SystemScriptGraph.NiagaraNodeAssignment_0.SetVariables_#1
 ]
}
emitters {
 BlowingParticles {
  emitterUpdateScript [
   &EmitterState(Life Cycle Mode=::NewEnumerator1, Loop Behavior=::NewEnumerator1, Loop Duration Mode=::NewEnumerator1, UseLoopDelay=false)
   /Game/Effects/Niagara/Modules/Logic/UpdateValueStateChange(InputIntParam:NiagaraInt32=&MakeIntFromFloat(Float=$User.BurstIndex, Conversion Mode=::NewEnumerator1))
   &self:TriggerSpawn(
    Spawn Count:NiagaraInt32=&MergeHelper_IntSwitch(
     0=12
     1=12
     2=7
     3=13
     4=15
     5=5
     6=4
     Index=&MakeIntFromFloat(Float=$User.BurstIndex, Conversion Mode=::NewEnumerator1)
     Num=::NewEnumerator5
    )
    Spawn Group:NiagaraInt32=&MakeIntFromFloat(Float=0, Conversion Mode=::NewEnumerator1)
    Trigger:NiagaraBool=&Boolean_AndOperation(Bool1=$Emitter.UpdateValueStateChange.Changed, Bool2=&SetBoolByFloatComparison(A=$System.Age, B=0.56))
   ) {module=TriggerSpawn}
   &self:TriggerSpawn(
    Spawn Count:NiagaraInt32=&MergeHelper_IntSwitch(
     0=2
     1=3
     2=1
     3=2
     4=3
     5=0
     6=0
     Index=&MakeIntFromFloat(Float=$User.BurstIndex, Conversion Mode=::NewEnumerator1)
     Num=::NewEnumerator5
    )
    Spawn Group:NiagaraInt32=&MakeIntFromFloat(Float=1, Conversion Mode=::NewEnumerator1)
    Trigger:NiagaraBool=&Boolean_AndOperation(Bool1=$Emitter.UpdateValueStateChange.Changed, Bool2=&SetBoolByFloatComparison(A=$System.Age, B=0.56))
   ) {module=TriggerSpawn001}
  ]
  particleSpawnScript [
   /SetVariables(Particles.VisibilityTag=&SpawnGroup()) {isSetParameters=true}
   &InitializeParticle(
    Color=&MergeHelper_ColorSwitch(
     0=&FNTimeOfDayColor(
      Night Color=<rgba 0.0546 0.1255 0.1719 1>
      Morning Color=<rgba 0.1361 0.2817 0.474 1>
      Day Color=<rgba 0.1185 0.3996 0.5833 1>
      Evening Color=<rgba 0.1361 0.2817 0.474 1>
     )
     1=&FNTimeOfDayColor(
      Night Color=<rgba 0.1068 0.1672 0.2135 1>
      Morning Color=<rgba 0.231 0.3123 0.4375 1>
      Day Color=<rgba 0.2891 0.4527 0.5781 1>
      Evening Color=<rgba 0.231 0.3123 0.4375 1>
     )
     2=?
     3=?
     Index=&SpawnGroup()
     Num=::NewEnumerator0
    )
    Lifetime=&MergeHelper_FloatSwitch(
     0=&RandomRangeFloat(Minimum=0.2, Maximum=0.6)
     1=0.3
     2=?
     3=?
     Index=&SpawnGroup()
     Num=::NewEnumerator0
    )
    Mass Max=1.5
    Mass Mode=::NewEnumerator3
    Sprite Rotation Mode=::NewEnumerator2
    Sprite Size Mode=::NewEnumerator3
    Uniform Sprite Size=&MergeHelper_FloatSwitch(
     0=&RandomRangeFloat(Minimum=0.5, Maximum=2)
     1=&RandomRangeFloat(Minimum=10, Maximum=20)
     2=?
     3=?
     Index=&SpawnGroup()
     Num=::NewEnumerator0
    )
    UsePositionOffset=false
    Write Material Random=true
   )
   &InitialMeshOrientation(Use Rotation Vector=false)
   &SocketLocation(
    SamplePosOffset=true
    Sampled Position Coordinate Space:ENiagaraCoordinateSpace=::World
    Sampled Position Offset=&MergeHelper_VectorSwitch(
     0=<20 0 0>
     1=<20 0 0>
     2=<20 0 0>
     3=<20 0 0>
     4=<20 0 0>
     5=<20 0 0>
     6=<20 0 0>
     Index=&MakeIntFromFloat(Float=$User.BurstIndex)
     Num=::NewEnumerator5
    )
    Socket Reader=$System.Socket Reader
    Socket Sampling Mode=::NewEnumerator2
    VelocitySpeedLimit=false
   )
   &AddVelocity(
    Cone Angle=26.8047
    Module.UseConeFalloff=false
    Module.UseVelDistribution=false
    Override Seed=false
    Rotation Mode=::NewEnumerator2
    Velocity Mode=::NewEnumerator2
    Velocity Speed=&MergeHelper_FloatSwitch(
     0=350
     1=350
     2=150
     3=350
     4=450
     5=0
     6=0
     Index=&MakeIntFromFloat(Float=$User.BurstIndex)
     Num=::NewEnumerator5
    )
    Yaw / Pitch / Roll:Vector3f=&MergeHelper_VectorSwitch(
     0=<0 42.0833 0>
     1=<0 140.528 0>
     2=<0 42.0833 0>
     3=<0 42.0833 0>
     4=<0 140.528 0>
     5=<0 0 0>
     6=<0 0 0>
     Index=&MakeIntFromFloat(Float=$User.BurstIndex)
     Num=::NewEnumerator5
    )
   )
   &ShapeLocation(
    Box Size=&MergeHelper_VectorSwitch(
     0=<10 10 10>
     1=<10 10 10>
     2=<10 10 10>
     3=<10 10 10>
     4=<10 10 10>
     5=<10 10 1>
     6=<10 10 1>
     Index=&MakeIntFromFloat(Float=$User.BurstIndex)
     Num=::NewEnumerator5
    )
    MaskSpawnGroup=false
    Override Seed=false
    Shape Primitive=::NewEnumerator2
    Surface Only=false
    UseOwnerScale=true
   )
  ]
  particleUpdateScript [
   &ParticleState()
   &SpriteRotationRate(Rotation Rate=&MaskFloatBySpawnGroup(Masked Float=0, Passthrough Float=&RandomRangeFloat(Minimum=-360, Maximum=360)))
   &GravityForce(Gravity=&MaskVectorBySpawnGroup(
     Masked Vector=&RandomRangeVector(Minimum=<0 0 -930.6271>, Maximum=<1 1 -1500>)
     Passthrough Vector=&RandomRangeVector(Minimum=<0 0 -1300>, Maximum=<1 1 -1500>)
    ))
   &DynamicMaterialParameters(
    Module.Index0Float0Write=true
    Module.Index0Float1Write=true
    Module.Index0Float2Write=true
    Module.Index0Float3Write=true
    Module.Index1Float0Write=true
    Module.Index1Float1Write=true
    Module.Index1Float2Write=true
    Module.Index1Float3Write=true
    Module.Index2Float0Write=true
    Module.Index2Float1Write=true
    Module.Index2Float2Write=true
    Module.Index2Float3Write=true
    Module.Index3Float0Write=true
    Module.Index3Float1Write=true
    Module.Index3Float2Write=true
    Module.Index3Float3Write=true
    Write Parameter Index 1=::NewEnumerator1
    Write Parameter Index 2=::NewEnumerator1
    Write Parameter Index 3=::NewEnumerator1
   )
   &InheritSourceMovement(
    Applied Position Scale=&MakeCustomVectorFromBool(
     Bool=&SetBoolByFloatComparison(A=&VectorLength(Vector=$User.OwnerLocalVelocity), B=147.6111)
     True Vector=<0.8 0.8 0.8>
     False Vector=<0 0 0>
    )
    ApplyForce=false
    ApplyPosition=true
    ApplyPositionFalloff=false
    Source Velocity:Vector3f=$User.OwnerLocalVelocity
   )
   &SolveForcesAndVelocity(Clamp Velocity=false, Limit Acceleration=false)
   &ScaleSpriteSizeBySpeed(
    Max Scale Factor=&MaskVector2dBySpawnGroup(Masked Vector2D=<1 4>, Passthrough Vector2D=<1 1>, Spawn Group Mask=0)
    Min Scale Factor=&MaskVector2dBySpawnGroup(Masked Vector2D=<1 2>, Passthrough Vector2D=<1 1>, Spawn Group Mask=0)
   )
   &ScaleSpriteSize(
    Scale Sprite Size Mode=::NewEnumerator0
    Uniform Scale Factor=&MergeHelper_FloatSwitch(
     0=&FloatFromCurve(FloatCurve=DataInterface() {propertyValues {Curve {keys [
          {arriveTangent=14.5835, interpMode=RCIM_Cubic, leaveTangent=14.5835, tangentMode=RCTM_User, time=0, value=0}
          {arriveTangent=0.1815, interpMode=RCIM_Cubic, leaveTangent=0.1815, tangentMode=RCTM_User, time=0.1, value=1}
          {arriveTangent=-0.0124, interpMode=RCIM_Cubic, leaveTangent=-0.0124, tangentMode=RCTM_User, time=0.6018, value=1}
          {arriveTangent=-4.3449, interpMode=RCIM_Cubic, leaveTangent=-4.3449, tangentMode=RCTM_User, time=1, value=0}
         ]}}})
     1=&FloatFromCurve(FloatCurve=DataInterface() {propertyValues {Curve {keys [
          {arriveTangent=14.5835, interpMode=RCIM_Cubic, leaveTangent=14.5835, tangentMode=RCTM_User, time=0, value=0}
          {arriveTangent=0.1815, interpMode=RCIM_Cubic, leaveTangent=0.1815, tangentMode=RCTM_User, time=0.1, value=1}
          {arriveTangent=-0.0124, interpMode=RCIM_Cubic, leaveTangent=-0.0124, tangentMode=RCTM_User, time=0.6018, value=1}
         ]}}})
     2=?
     3=?
     Index=&SpawnGroup()
     Num=::NewEnumerator0
    )
   )
   &SubUVAnimation(
    Evaluation Type=::NewEnumerator0
    "Sprite Renderer "=DataInterface() {propertyValues {SpriteRenderer=&self:BlowingParticles.NiagaraSpriteRendererProperties_1}}
    SubUV Animation Mode:ENiagara_SubUVLookupModeV2=::NewEnumerator2
    UseEndFrame=false
    UseStartFrame=false
   )
   &ScaleColor(
    Scale Alpha=&FloatFromCurve(FloatCurve=DataInterface() {propertyValues {Curve {keys [<at 0 1>,<at 1 0>]}}})
    ScaleA=true
    ScaleRGB=true
   )
  ]
  rendererClasses [&NiagaraSpriteRendererProperties]
  renderers [
   {rendererClass=&NiagaraSpriteRendererProperties}
   {rendererClass=&NiagaraSpriteRendererProperties, rendererIndex=1}
  ]
 }
 Trail {
  emitterSpawnScript [/SetVariables(Emitter.WetColor:LinearColor=&MakeCustomLinearColorFromBool(Bool=$User.bIsFrontEndPreview, True Color=<rgba 0.0879 0.2947 0.4688 1>, False Color=<rgba 0.097 0.1496 0.1939 0.2251>)) {isSetParameters=true}]
  emitterUpdateScript [
   &EmitterState(Life Cycle Mode=::NewEnumerator1, Loop Behavior=::NewEnumerator1, Loop Duration Mode=::NewEnumerator1, UseLoopDelay=false)
   &SpawnRate(
    SpawnGroup=1
    SpawnRate=&MakeCustomFloatFromBool(Bool=$User.bIsFrontEndPreview, True Float=0, False Float=$User.SpawnCurve)
    Use Spawn Probability=false
   )
   &SpawnRate(SpawnRate=$User.SpawnCurve, Use Spawn Probability=false) {module=SpawnRate002}
   &SpawnRate(
    SpawnGroup=2
    SpawnRate=&MakeCustomFloatFromBool(Bool=$User.bIsFrontEndPreview, True Float=$User.SpawnCurve, False Float=0)
    Use Spawn Probability=false
   ) {module=SpawnRate001}
  ]
  particleSpawnScript [
   /SetVariables(Particles.VisibilityTag=&SpawnGroup()) {isSetParameters=true}
   &InitializeParticle(
    Color=&SelectLinearColorFromArray(
     Array Sampling Mode=::NewEnumerator0
     Direct Array Index=?
     Color Selection Array=DataInterface() {propertyValues {ColorData [<rgba 0.7504 0.8698 0.9479 1>, <rgba 0.7899 0.9479 0.9181 1>]}}
     Random Seed=0
    )
    Lifetime=&MergeHelper_FloatSwitch(
     0=2
     1=&RandomRangeFloat(Minimum=1, Maximum=0.5)
     2=?
     3=?
     Index=&SpawnGroup()
     Num=::NewEnumerator0
    )
    Sprite Rotation Mode=::NewEnumerator2
    Sprite Size Mode=::NewEnumerator1
    Uniform Sprite Size Max=45
    Uniform Sprite Size Min=65
    UsePositionOffset=false
    Write Material Random=true
   )
   /SetVariables(Particles.Sparkles Lifetime:NiagaraFloat=&RandomRangeFloat(Minimum=0.3, Maximum=0.4)) {isSetParameters=true}
   /SetVariables(
    Particles.DecalOrientation:Quat4f=/Game/Effects/Niagara/Modules/MergeEmitters/MergeHelper_QuatSwitch(
     0:Quat4f=<0 1 0 1>
     1:Quat4f=<0.2 1 0 1>
     2:Quat4f=<0.4 1 0 1>
     3:Quat4f=<0.6 1 0 1>
     4:Quat4f=<0.8 1 0 1>
     5:Quat4f=?
     6:Quat4f=?
     7:Quat4f=?
     8:Quat4f=?
     9:Quat4f=?
     Index:NiagaraInt32=&SelectIntFromArray(Direct Array Index=0, Int Selection Array=DataInterface() {propertyValues {IntData [0,1,2,3,4]}})
     Num:MergeHelper_SwitcherEnum=::NewEnumerator3
    )
    Particles.DecalSize:Vector3f=<200 56.847 72.9045>
   ) {isSetParameters=true}
   &SocketLocation(
    SamplePosOffset=true
    Sampled Position Offset=&MaskVectorBySpawnGroup(Masked Vector=<0 0 -30>, Passthrough Vector=<0 0 0>, Spawn Group Mask=2)
    Sampled Position Offset Coordinate Space:ENiagaraSkelMeshTransforms=::NewEnumerator2
    Socket Reader=$System.Socket Reader
    Socket Sampling Mode=::NewEnumerator2
    VelocitySpeedLimit=false
   )
   /SetVariables(
    Particles.DecalPos:NiagaraPosition=&ConvertVectorToPosition(
     Input Position=Linked() {linkedVariable {name=Output.SocketLocation.SampledPosition, type=NiagaraPosition}}
     Conversion Operation:ENiagara_LWCConvertVecToPos=::NewEnumerator0
    )
    Particles.SparklePos:NiagaraPosition=&AddVectorToPosition(
     Position=&MakeVector(
      X=&MakeFloatFromVector(Vector=Linked() {linkedVariable {name=Output.SocketLocation.SampledPosition, type=NiagaraPosition}}, Channel=::NewEnumerator0)
      Y=&MakeFloatFromVector(Vector=Linked() {linkedVariable {name=Output.SocketLocation.SampledPosition, type=NiagaraPosition}}, Channel=::NewEnumerator1)
      Z=&ClampFloat(
       Float=&MakeFloatFromVector(Vector=Linked() {linkedVariable {name=Output.SocketLocation.SampledPosition, type=NiagaraPosition}}, Channel=::NewEnumerator2)
       Min=&MakeFloatFromVector(Vector=Linked() {linkedVariable {name=Output.SocketLocation.SampledPosition, type=NiagaraPosition}}, Channel=::NewEnumerator2)
       Max=&Add_Float(
        A=-25
        B=&MakeFloatFromVector(Vector=Linked() {linkedVariable {name=Output.SocketLocation.SampledPosition, type=NiagaraPosition}}, Channel=::NewEnumerator2)
       )
      )
     )
     Vector=&RandomRangeVector(Minimum=<-20 -20 0>, Maximum=<20 20 0>)
    )
   ) {isSetParameters=true}
   &CameraOffset(Camera Offset Amount=&MaskFloatBySpawnGroup(Masked Float=20, Passthrough Float=0))
   /SetVariables(Particles.NewPos:NiagaraPosition=&AddVectorToPosition(
     Position=$Output.SocketLocation.SampledPosition
     Vector=&RandomRangeVector(Minimum=<-20 -20 0>, Maximum=<20 20 0>)
    )) {isSetParameters=true}
   &ShapeLocation(
    Box Size=<20 20 1>
    MaskSpawnGroup=true
    Override Seed=false
    Shape Primitive=::NewEnumerator2
    Spawn Group Mask=2
    Surface Only=false
    UseOwnerScale=true
   )
   &AddVelocity(Velocity=&MakeCustomVectorFromBool(Bool=$User.bIsFrontEndPreview, True Vector=<0 -70 0>, False Vector=<0 0 0>))
  ]
  particleUpdateScript [
   &ParticleState()
   &SolveForcesAndVelocity(Clamp Velocity=false, Limit Acceleration=false)
   &self:Slope_Position_01(
    Fallback Position:NiagaraPosition=&AddVectorToPosition(
     Position=&MakeVector(
      X=&MakeFloatFromVector(Vector=Linked() {linkedVariable {name=Output.SocketLocation.SampledPosition, type=NiagaraPosition}}, Channel=::NewEnumerator0)
      Y=&MakeFloatFromVector(Vector=Linked() {linkedVariable {name=Output.SocketLocation.SampledPosition, type=NiagaraPosition}}, Channel=::NewEnumerator1)
      Z=&ClampFloat(
       Float=&MakeFloatFromVector(Vector=Linked() {linkedVariable {name=Output.SocketLocation.SampledPosition, type=NiagaraPosition}}, Channel=::NewEnumerator2)
       Min=&MakeFloatFromVector(Vector=Linked() {linkedVariable {name=Output.SocketLocation.SampledPosition, type=NiagaraPosition}}, Channel=::NewEnumerator2)
       Max=&Add_Float(
        A=-25
        B=&MakeFloatFromVector(Vector=Linked() {linkedVariable {name=Output.SocketLocation.SampledPosition, type=NiagaraPosition}}, Channel=::NewEnumerator2)
       )
      )
     )
     Vector=&RandomRangeVector(Minimum=<-20 -20 0>, Maximum=<20 20 0>)
    )
    Height Adjustment/Offset=1
    Slope Position:NiagaraPosition=$Particles.NewPos
    Trace Length=120
    Trace Origin Height=85
    bIsFrontEnd:NiagaraBool=$User.bIsFrontEndPreview
   ) {module=Slope_Position_01}
   /SetVariables(Particles.DecalColor:LinearColor=&MakeLinearColorFromVectorAndFloat(
     "Vector (RGB)"=&MakeVector(
      X=&FloatFromCurve(FloatCurve=DataInterface() {propertyValues {Curve {keys [{arriveTangent=2.6346, interpMode=RCIM_Cubic, leaveTangent=2.6346, tangentMode=RCTM_User, time=0, value=-1}, <at 1 0>]}}})
      Y=0
      Z=0
     )
     "Float (Alpha)"=&FloatFromCurve(FloatCurve=DataInterface() {propertyValues {Curve {
         defaultValue=1
         keys [
          {arriveTangent=5.9515, interpMode=RCIM_Cubic, leaveTangent=5.9515, tangentMode=RCTM_User, time=0, value=0}
          {interpMode=RCIM_Cubic, tangentMode=RCTM_User, time=0.0975, value=1}
          {arriveTangentWeight=0.2201, interpMode=RCIM_Cubic, leaveTangentWeight=0.2332, tangentMode=RCTM_User, time=0.5259, value=1}
          {arriveTangent=0.086, interpMode=RCIM_Cubic, leaveTangent=0.086, tangentMode=RCTM_User, time=1, value=0}
         ]
        }}})
    )) {isSetParameters=true}
   /SetVariables(Particles.SparklesSize:Vector2f=&Multiply_Vector2D_ByFloat(
     Float=&FloatFromCurve(
      FloatCurve=DataInterface() {propertyValues {Curve {
         defaultValue=0
         keys [
          {interpMode=RCIM_Cubic, tangentMode=RCTM_User, time=0, value=0}
          {interpMode=RCIM_Cubic, tangentMode=RCTM_User, time=0.5, value=1}
          {tangentMode=RCTM_User, time=1, value=0}
         ]
        }}}
      Scale Curve=&RandomRangeFloat(Minimum=17, Maximum=25)
     )
     Vector2D=<1 1>
    )) {isSetParameters=true}
   &ScaleSpriteSize(
    Uniform Curve Index=$Particles.Update Sparkles Age
    Uniform Curve Sprite Scale=DataInterface() {propertyValues {Curve {
       defaultValue=0
       keys [
        {interpMode=RCIM_Cubic, tangentMode=RCTM_User, time=0, value=0}
        {interpMode=RCIM_Cubic, tangentMode=RCTM_User, time=0.5, value=1}
        {tangentMode=RCTM_User, time=1, value=0}
       ]
      }}}
   )
   &SpriteRotationRate(Rotation Rate=&SelectFloatFromArray(
     Array Sampling Mode=::NewEnumerator0
     Direct Array Index=?
     Float Selection Array=DataInterface() {propertyValues {FloatData [-400,400]}}
     Random Seed=0
    ))
   &BlendSpace("Position Blend (World <-> Local)":NiagaraFloat=&MaskFloatBySpawnGroup(
     Masked Float=&MakeCustomFloatFromBool(
      Bool=&SetBoolByFloatComparison(A=&VectorLength(Vector=$Engine.Owner.Velocity), B=400)
      True Float=1
      False Float=0
     )
     Passthrough Float=0
     Translation Contribution:NiagaraFloat=&MaskFloatBySpawnGroup(
      Masked Float=&MakeCustomFloatFromBool(
       Bool=&SetBoolByFloatComparison(A=&VectorLength(Vector=$Engine.Owner.Velocity), B=400)
       True Float=1
       False Float=0
      )
      Passthrough Float=0
     )
     Rotation Contribution:NiagaraFloat=&MaskFloatBySpawnGroup(
      Masked Float=&MakeCustomFloatFromBool(
       Bool=&SetBoolByFloatComparison(A=&VectorLength(Vector=$Engine.Owner.Velocity), B=400)
       True Float=1
       False Float=0
      )
      Passthrough Float=0
     )
    )) {enabled=false}
  ]
  rendererClasses [&NiagaraDecalRendererProperties, &NiagaraSpriteRendererProperties]
  renderers [
   {rendererClass=&NiagaraDecalRendererProperties, rendererIndex=0}
   {rendererClass=&NiagaraSpriteRendererProperties, rendererIndex=1}
   {rendererClass=&NiagaraSpriteRendererProperties, rendererIndex=2}
  ]
 }
}
summary {
 systemName=NS_Emote_MopSwirl_MopPuddleAndFlicks
 emitters [
  {emitterName=Trail, rendererClasses [&NiagaraDecalRendererProperties, &NiagaraSpriteRendererProperties]}
  {emitterName=BlowingParticles, rendererClasses [&NiagaraSpriteRendererProperties]}
 ]
 userVariables [
  User.bIsFrontEnd {type=NiagaraBool}
  User.bIsFrontEndPreview {type=NiagaraBool}
  User.BurstIndex {type=NiagaraFloat, defaultValue=1, description=""}
  User.OwnerLocalVelocity {type=Vector3f, description="", defaultValue {struct=Vector3f, value=<0 0 0>}}
  User.SpawnCurve {type=NiagaraFloat, defaultValue=0, description=""}
 ]
}
