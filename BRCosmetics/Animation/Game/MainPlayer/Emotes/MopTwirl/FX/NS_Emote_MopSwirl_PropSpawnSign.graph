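# Niagara system for the MopTwirl emote, named as the effect for the spawn of a sign prop.
# It combines three emitters, SparklesLWE, AirglowLWE and SmokeLWE, whose LWE suffix suggests Lightweight Emitters; no renderers are listed for them.
# System logic uses SystemState and a set-variables assignment.
# It exposes a float user parameter, NormalizedNotifyProgress, apparently meant to receive animation notify progress.
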NiagaraSystem /BRCosmetics/Animation/Game/MainPlayer/Emotes/MopTwirl/FX/NS_Emote_MopSwirl_PropSpawnSign
dependencies {
 usedDataInterfaces []
 usedDynamicInputs []
 usedModules [
  &self:SystemSpawnScript.SystemScriptSource.SystemScriptGraph.NiagaraNodeAssignment_0.SetVariables_#1
  &SystemState
 ]
}
emitters {AirglowLWE {rendererClasses [], renderers []}, SmokeLWE {rendererClasses [], renderers []}, SparklesLWE {rendererClasses [], renderers []}}
summary {
 systemName=NS_Emote_MopSwirl_PropSpawnSign
 emitters [{emitterName=SparklesLWE, rendererClasses []}, {emitterName=AirglowLWE, rendererClasses []}, {emitterName=SmokeLWE, rendererClasses []}]
 userVariables [User.NormalizedNotifyProgress {type=NiagaraFloat, defaultValue=0, description=""}]
}
